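; Creative pinball bumper device that knocks players, vehicles and physics objects sideways or upward, awarding score for players and vehicle passengers.
; Trigger overlaps or activation events send server gameplay events to its equipped ability set; launch events apply configurable knockback, blending radial and reflected-velocity directions for side hits.
; Gameplay cues drive bounce timelines, sound and glow through cached dynamic materials.
; Replicated disabled state and an optional reset timer handle cooldown.
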(event OnComponentBeginOverlap(UpTrigger)
  (params (OverlappedComponent :type Primitive%20Component%20Object%20Reference) (OtherActor :type Actor%20Object%20Reference) (OtherComp :type Primitive%20Component%20Object%20Reference) (OtherBodyIndex :type Integer) (bFromSweep :type Boolean) (SweepResult :type Hit%20Result%20Structure%20%28by%20ref%29))
  ;[1 1 1] Bumper is Triggered
  (CallFunction|AttemptKnock OtherActor true))

(event Custom|Disable%20FX
  ;[1 1 1] Resetting the Top Mesh Element
  (bind (_direction _pinballbumper_03_z_offset) (Timeline|LowerTop PlayFromStart)
    (:Update
      (Transformation|SetRelativeLocation (Variables|Default|GetOC_PinballBumper_03) (Math|Vector|MakeVector 0.0 0.0 _pinballbumper_03_z_offset) false false))))

(event OnComponentBeginOverlap(Trigger)
  (params (OverlappedComponent :type Primitive%20Component%20Object%20Reference) (OtherActor :type Actor%20Object%20Reference) (OtherComp :type Primitive%20Component%20Object%20Reference) (OtherBodyIndex :type Integer) (bFromSweep :type Boolean) (SweepResult :type Hit%20Result%20Structure%20%28by%20ref%29))
  ;[1 1 1] Bumper is Triggered
  (CallFunction|AttemptKnock OtherActor))

(event Custom|Reset%20FX
  ;[1 1 1] Resetting the Top Mesh Element
  (Audio|PlaySoundatLocation (Variables|Sound|GetBumper%20Ready%20Sound) (Transformation|GetActorLocation))
  (bind (_direction _newtrack_0) (Timeline|RaiseTop PlayFromStart)
    (:Update
      (Transformation|SetRelativeLocation (Variables|Default|GetOC_PinballBumper_03) (Math|Vector|MakeVector 0.0 0.0 _newtrack_0) false false))))

(event Custom|Enable%20Event
  (params (Instigator :type Controller%20Object%20Reference))
  ;[0.6337 0 1] Gameplay Functions
  (CallFunction|ResetBumper))

(event Custom|Activate%20Event
  (params (Instigator :type Controller%20Object%20Reference))
  ;[0.6337 0 1] Gameplay Functions
  (CallFunction|AttemptKnock 0 true))

(event Custom|Disable%20Event
  (params (Instigator :type Controller%20Object%20Reference))
  ;[0.6337 0 1] Gameplay Functions
  (CallFunction|DisableBumper))

(event Custom|LaunchPhysicsObjectUp
  (params (PhysicsObject :type Fort%20Physics%20Pawn%20Object%20Object%20Reference))
  ;[1 0.7546 0] Launch Objects
  (Utilities|IsValid PhysicsObject
    (:Is%20Valid
      (Physics|AddImpulse (Variables|/Script/FortniteGame.FortPhysicsPawnObject|GetPhysicsMeshComponent PhysicsObject) (* (Transformation|GetUpVector (Variables|Default|GetBody)) (CallFunction|GetKnockbackValue)) "None" true))))

(event EventBeginPlay
  ;[1 1 1] Initialization
  (|CallParentFunction)
  (Networking|SwitchHasAuthority
    (:Authority
      (Ability|EquipFortAbilitySet self "/Game/Creative/Devices/Pinball_Bumper/GAS_Creative_PinballBumper.GAS_Creative_PinballBumper"))))

(event Custom|LaunchPhysicsObject
  (params (Target :type Fort%20Physics%20Pawn%20Object%20Object%20Reference) (ImpactPoint :type Vector))
  ;[1 0.7546 0] Launch Objects
  (Utilities|IsValid Target
    (:Is%20Valid
      (bind _vector (CallFunction|GenericKnockbackActions Target ImpactPoint))
      (PhysicsObject|ApplyImpulse Target self _vector))))

(event MinigameStateChangedDelegate(FortMinigameProgress)
  (params (Minigame :type Fort%20Minigame%20Object%20Reference) (NewMinigameState :type EFortMinigameState%20Enum :default "PreGame"))
  ;[1 1 1] Initialization
  (if (Utilities|Enum|Equal(Enum) NewMinigameState "PreGame")
    (CallFunction|BumperEnableOnStart)))

(event Custom|LaunchVehicle
  (params (Target :type Fort%20Athena%20Vehicle%20Object%20Reference) (ImpactPoint :type Vector))
  ;[1 0.7546 0] Launch Objects
  (Utilities|IsValid Target
    (:Is%20Valid
      (Score|AwardScoreVehiclePassengers Target)
      (bind _vector (CallFunction|GenericKnockbackActions Target ImpactPoint))
      (Audio|BroadcastAbilityImpactatLocation 0 Target _vector 777.0))))

(event Custom|LaunchVehicleUp
  (params (Target :type Fort%20Athena%20Vehicle%20Object%20Reference))
  ;[1 0.7546 0] Launch Objects
  (Utilities|IsValid Target
    (:Is%20Valid
      (Score|AwardScoreVehiclePassengers Target)
      (Audio|BroadcastAbilityImpactatLocation 0 Target (* (Transformation|GetUpVector (Variables|Default|GetBody)) (CallFunction|GetKnockbackValue)) 777.0))))

(event OnPropertiesLoaded(ToyOptionsComponent)
  (bind _output (Utilities|ResolveSoftReference (Utilities|Array|Get(acopy) (Variables|UserKnobs|GetBumperMaterials) (Variables|UserKnobs|GetMaterialIndex))))
  (bind _materialindex (Variables|UserKnobs|GetMaterialIndex))
  (bind _bumpermaterials (Variables|UserKnobs|GetBumperMaterials))
  (bind _output_2 (Utilities|Array|Get(acopy) _bumpermaterials _materialindex))
  ;[1 1 1] Initialization
  (if (Variables|UserOptions|Advanced|GetDisplayScoreUpdateOnHUD)
    (Class|CreativeUIBroadcastComponent|Initialize (Variables|Default|GetCreative_HUDBroadcast_Component) (Variables|Default|GetFortMinigameProgress) true)
    (else
      _))
  (Utilities|FlowControl|Sequence
    (:then_0
      (if (Utilities|Array|IsValidIndex _bumpermaterials _materialindex)
        (if (Utilities|IsValidSoftObjectReference _output_2)
          (Utilities|IsValid _output
            (:Is%20Valid
              (CallFunction|UpdateParentMaterial _output))
            (:Is%20Not%20Valid
              (bind _object (Utilities|AsyncLoadAsset _output_2)
                (:Completed
                  (bind _asmaterial_interface (Utilities|Casting|CastToMaterialInterface _object))
                  (CallFunction|UpdateParentMaterial _asmaterial_interface))))))))
    (:then_1
      (CallFunction|BumperEnableOnStart))))

(event OnAnyPropertyChanged(ToyOptionsComponent)
  (bind _output (Utilities|ResolveSoftReference (Utilities|Array|Get(acopy) (Variables|UserKnobs|GetBumperMaterials) (Variables|UserKnobs|GetMaterialIndex))))
  (bind _materialindex (Variables|UserKnobs|GetMaterialIndex))
  (bind _bumpermaterials (Variables|UserKnobs|GetBumperMaterials))
  (bind _output_2 (Utilities|Array|Get(acopy) _bumpermaterials _materialindex))
  ;[1 1 1] Initialization
  (if (Variables|UserOptions|Advanced|GetDisplayScoreUpdateOnHUD)
    (Class|CreativeUIBroadcastComponent|Initialize (Variables|Default|GetCreative_HUDBroadcast_Component) (Variables|Default|GetFortMinigameProgress) true)
    (else
      _))
  (Utilities|FlowControl|Sequence
    (:then_0
      (if (Utilities|Array|IsValidIndex _bumpermaterials _materialindex)
        (if (Utilities|IsValidSoftObjectReference _output_2)
          (Utilities|IsValid _output
            (:Is%20Valid
              (CallFunction|UpdateParentMaterial _output))
            (:Is%20Not%20Valid
              (bind _object (Utilities|AsyncLoadAsset _output_2)
                (:Completed
                  (bind _asmaterial_interface (Utilities|Casting|CastToMaterialInterface _object))
                  (CallFunction|UpdateParentMaterial _asmaterial_interface))))))))
    (:then_1
      (CallFunction|BumperEnableOnStart))))

(event Custom|LaunchPawn
  (params (Pawn :type Fort%20Pawn%20Object%20Reference) (ImpactPoint :type Vector))
  ;[1 0.7546 0] Launch Objects
  (Utilities|IsValid Pawn
    (:Is%20Valid
      (Score|AwardScore Pawn)
      (bind _vector (CallFunction|GenericKnockbackActions Pawn ImpactPoint))
      (Pawn|Character|LaunchCharacterJump Pawn _vector false true (not (Variables|UserKnobs|GetbAllowFallDamage)) false))))

(event EditorPreview|EventSetupForEditorPreview
  (bind _output (Utilities|ResolveSoftReference (Utilities|Array|Get(acopy) (Variables|UserKnobs|GetBumperMaterials) (Variables|UserKnobs|GetMaterialIndex))))
  (bind _materialindex (Variables|UserKnobs|GetMaterialIndex))
  (bind _bumpermaterials (Variables|UserKnobs|GetBumperMaterials))
  (bind _output_2 (Utilities|Array|Get(acopy) _bumpermaterials _materialindex))
  ;[1 1 1] Initialization
  (Utilities|FlowControl|Sequence
    (:then_0
      (if (Utilities|Array|IsValidIndex _bumpermaterials _materialindex)
        (if (Utilities|IsValidSoftObjectReference _output_2)
          (Utilities|IsValid _output
            (:Is%20Valid
              (CallFunction|UpdateParentMaterial _output))
            (:Is%20Not%20Valid
              (bind _object (Utilities|AsyncLoadAsset _output_2)
                (:Completed
                  (bind _asmaterial_interface (Utilities|Casting|CastToMaterialInterface _object))
                  (CallFunction|UpdateParentMaterial _asmaterial_interface))))))))
    (:then_1
      (CallFunction|BumperEnableOnStart))))

(event Custom|LaunchPlayerUp
  (params (Pawn :type Fort%20Pawn%20Object%20Reference))
  (bind (_vehicleknockbackvalue _pawnknockbackvalue) (CallFunction|GetKnockbackValue))
  ;[1 0.7546 0] Launch Objects
  (Utilities|IsValid Pawn
    (:Is%20Valid
      (Score|AwardScore Pawn)
      (Pawn|Character|LaunchCharacterJump Pawn (* (Transformation|GetUpVector (Variables|Default|GetBody)) _pawnknockbackvalue) false true (not (Variables|UserKnobs|GetbAllowFallDamage)) false))))

(event GameplayCue|GameplayCue.Abilities.Activation.Devices.PinballBumper.KnockSide
  (params (EventType :type EGameplayCueEvent%20Enum :default "OnActive") (Parameters :type Gameplay%20Cue%20Parameters%20Structure%20%28by%20ref%29))
  ;[1 1 1] Cosmetic Effects on Hit
  (Networking|CanExecuteCosmeticEvents
    (:True
      (CallFunction|SwitchToDyamicMaterial)
      (bind _material_instance_dynamic (Variables|Default|GetMaterial%20Instance%20Dynamic)
        (:then
          (Rendering|Material|SetScalarParameterValue _material_instance_dynamic "GlowOn" 1.0 :!class /Script/Engine.MaterialInstanceDynamic))
        (:else
          _))
      (bind _returnvalue (Effects|SpawnEmitteratLocation "/Game/Effects/Fort_Effects/Effects/Creative/Props/P_Creative_Pinball_Bumper.P_Creative_Pinball_Bumper" (Transformation|GetWorldLocation (Variables|Default|GetBody))))
      (Effects|Components|ParticleSystem|SetColorParameter _returnvalue "LightColor" (Utilities|Array|Get(acopy) (Variables|UserKnobs|GetFX_Colors) (Variables|UserKnobs|GetMaterialIndex)))
      (Audio|PlaySoundatLocation (Variables|Sound|GetBounce%20Sound) (Transformation|GetActorLocation))
      (bind (_direction _pinballbumper_02_scale) (Timeline|Bounce PlayFromStart)
        (:Update
          (Transformation|SetRelativeScale3D (Variables|Default|GetOC_PinballBumper_02) (Math|Vector|MakeVector _pinballbumper_02_scale _pinballbumper_02_scale 1.0)))
        (:Finished
          (bind _material_instance_dynamic_2 (Variables|Default|GetMaterial%20Instance%20Dynamic)
            (:then
              (Rendering|Material|SetScalarParameterValue _material_instance_dynamic_2 "GlowOn" 0.0 :!class /Script/Engine.MaterialInstanceDynamic)
              (CallFunction|SwitchToConstantMaterial))
            (:else
              _))
          (if (not (> (Variables|UserKnobs|GetResetTime) 0.0))
            (CallFunction|ResetFX)))))))

(event GameplayCue|GameplayCue.Abilities.Activation.Devices.PinballBumper.KnockUp
  (params (EventType :type EGameplayCueEvent%20Enum :default "OnActive") (Parameters :type Gameplay%20Cue%20Parameters%20Structure%20%28by%20ref%29))
  ;[1 1 1] Cosmetic Effects on Hit
  (Networking|CanExecuteCosmeticEvents
    (:True
      (Audio|PlaySoundatLocation (Variables|Sound|GetBounce%20Sound) (Transformation|GetActorLocation))
      (CallFunction|SwitchToDyamicMaterial)
      (bind _material_instance_dynamic (Variables|Default|GetMaterial%20Instance%20Dynamic)
        (:then
          (Rendering|Material|SetScalarParameterValue _material_instance_dynamic "GlowOn" 1.0 :!class /Script/Engine.MaterialInstanceDynamic))
        (:else
          _))
      (bind (_direction _pinballbumper_03_z_offset) (Timeline|BounceUp PlayFromStart)
        (:Update
          (Transformation|SetRelativeLocation (Variables|Default|GetOC_PinballBumper_03) (Math|Vector|MakeVector 0.0 0.0 _pinballbumper_03_z_offset) false false))
        (:Finished
          (bind _material_instance_dynamic_2 (Variables|Default|GetMaterial%20Instance%20Dynamic)
            (:then
              (Rendering|Material|SetScalarParameterValue _material_instance_dynamic_2 "GlowOn" 0.0 :!class /Script/Engine.MaterialInstanceDynamic)
              (CallFunction|SwitchToConstantMaterial))
            (:else
              _))
          (if (not (> (Variables|UserKnobs|GetResetTime) 0.0))
            (CallFunction|ResetFX)))))))

(event Custom|Switch%20to%20Dyamic%20Material
  ;[1 1 1] Material Changes
  (Networking|CanExecuteCosmeticEvents
    (:True
      (Variables|Default|GetMaterial%20Instance%20Dynamic
        (:then
          _)
        (:else
          (bind _material_cache (Variables|Default|GetMaterial%20Cache)
            (:then
              (bind (_outmaterialinstance _returnvalue) (Class|FortMaterialCache|ReserveMaterial _material_cache))
              (Variables|Default|SetMaterial%20Instance%20Dynamic _outmaterialinstance)
              (if _returnvalue
                _))
            (:else))))
      (Rendering|Material|SetMaterial (Variables|Default|GetOC_PinballBumper_03) 0 (Variables|Default|GetMaterial%20Instance%20Dynamic)))))

(event Custom|Switch%20to%20Constant%20Material
  ;[1 1 1] Material Changes
  (Networking|CanExecuteCosmeticEvents
    (:True
      (Rendering|Material|SetMaterial (Variables|Default|GetOC_PinballBumper_03) 0 (Variables|Default|GetParentMaterial))
      (bind _material_cache (Variables|Default|GetMaterial%20Cache)
        (:then
          (Class|FortMaterialCache|ReleaseMaterial _material_cache (Variables|Default|GetMaterial%20Instance%20Dynamic)))
        (:else)))))

(fn AttemptKnock
  (params (Actor :type Actor%20Object%20Reference) (Up :type Boolean))
  (Networking|SwitchHasAuthority
    (:Authority
      (Utilities|CanTrigger Actor
        (:True
          (Ability|SendGameplayEventtoActor self (select Up (Variables|Default|GetKnockUpCue) (Variables|Default|GetKnockSideCue)))
          (bind _currentminigame (Variables|/Script/FortniteGame.FortMinigameLogicComponent|GetCurrentMinigame (Variables|Default|GetFortMinigameProgress))
            (:then
              (Minigame|IsInProgress _currentminigame
                (:Yes
                  (Utilities|FlowControl|Sequence
                    (:then_0
                      (bind _ascontroller (Utilities|Casting|CastToController Actor)
                        (:then
                          (LocalVariables|SetInstigatorController _ascontroller))
                        (:CastFailed
                          (bind _aspawn (Utilities|Casting|CastToPawn Actor))
                          (LocalVariables|SetInstigatorController (Pawn|GetController _aspawn)))))
                    (:then_1
                      (GameplayEvent|BroadcastEvent (Variables|UserOptions|Events|GetOn%20Activated) 0 (Utilities|Struct|MakeGameplayEventActivated (LocalVariables|GetInstigatorController))))))))
            (:else)))))))

(fn AwardScore
  (params (Pawn :type Pawn%20Object%20Reference))
  (bind _returnvalue (Fortnite|GetFortPlayerControllerfromActor Pawn))
  (Networking|SwitchHasAuthority
    (:Authority
      (Utilities|IsValid _returnvalue
        (:Is%20Valid
          (bind _returnvalue_2 (CreativeDevices|GetMinigameStatsComp _returnvalue "Athena"))
          (Utilities|IsValid _returnvalue_2
            (:Is%20Valid
              (Class|FortMinigameStatsComponent|SendStatEventwithTags _returnvalue_2 _returnvalue "MinigameDynamicEvent" self (GameplayTags|MakeGameplayTagContainerfromTag "(TagName=\"Minigame.Stat.Score.Pinball\")") (Class|FortMinigameStatsComponent|GetSourceandContextTags _returnvalue_2 _returnvalue) (.OutContextTags (Class|FortMinigameStatsComponent|GetSourceandContextTags _returnvalue_2 _returnvalue)) (Variables|UserKnobs|GetScoreValueSetting))
              (Class|CreativeHUDBroadcastComponent|DoScoreEvent (Variables|Default|GetCreative_HUDBroadcast_Component) _returnvalue (Variables|UserKnobs|GetScoreValueSetting) (Variables|UserOptions|Advanced|GetDisplayScoreUpdateOnHUD)))))))))

(fn AwardScoreVehiclePassengers
  (params (Vehicle :type Fort%20Athena%20Vehicle%20Object%20Reference))
  (Networking|SwitchHasAuthority
    (:Authority
      (for _array_element (Vehicle|GetAllPassengers Vehicle)
        (Utilities|IsValid _array_element
          (:Is%20Valid
            (Score|AwardScore _array_element)))))))

(fn BlueprintCanInteract
  (params (InteractingPawn :type Fort%20Pawn%20Object%20Reference) (InteractionBeingAttempted :type EInteractionBeingAttempted%20Enum :default "FirstInteraction") (InteractionType :type TInteractionType%20Enum :default "IT_NoInteraction"))
  (bind _returnvalue (Pawn|GetController InteractingPawn))
  (bind _fortminigameprogress (Variables|Default|GetFortMinigameProgress))
  (bind _currentminigame (Variables|/Script/FortniteGame.FortMinigameLogicComponent|GetCurrentMinigame _fortminigameprogress))
  (Utilities|FlowControl|Sequence
    (:then_0
      (Utilities|IsValid (Variables|/Script/FortniteGame.FortActorOptionsComponent|GetInternal_PlayerOptions (Variables|Default|GetToyOptionsComponent))
        (:Is%20Valid
          (Utilities|IsValid InteractingPawn
            (:Is%20Valid
              (Utilities|IsValid _returnvalue
                (:Is%20Valid
                  (bind _asfort_player_controller_athena (Utilities|Casting|CastToFortPlayerControllerAthena _returnvalue))
                  (if (Creative|CanPlayerCreateinCurrentExperience _asfort_player_controller_athena)
                    (Utilities|IsValid _currentminigame
                      (:Is%20Valid
                        (LocalVariables|SetbInteractSuccess (not (Minigame|HasStarted _currentminigame :!class /Script/FortniteGame.FortMinigame))))
                      (:Is%20Not%20Valid
                        (LocalVariables|SetbInteractSuccess true)))))))))))
    (:then_1
      (return (LocalVariables|GetbInteractSuccess)))))

(fn BlueprintGetInteractionString
  (params (InteractingPawn :type Fort%20Pawn%20Object%20Reference) (InteractionBeingAttempted :type EInteractionBeingAttempted%20Enum :default "FirstInteraction"))
  (return "NSLOCTEXT(\"[EFE666AA45CA7B4C4170BE8E6727EBB9]\", \"6CC4539B483DAE156FF053A62053652C\", \"Customize\")"))

(fn BlueprintOnLocalInteract
  (params (InteractingPawn :type Fort%20Player%20Pawn%20Object%20Reference) (InteractionBeingAttempted :type EInteractionBeingAttempted%20Enum :default "FirstInteraction"))
  (return false))

(fn BumperEnableonStart
  (Networking|SwitchHasAuthority
    (:Authority
      (if (Variables|UserKnobs|GetEnabledOnStart)
        (CallFunction|ResetBumper)
        (else
          (CallFunction|DisableBumper))))))

(fn CalculateLaunchVector
  (params (InputPin :type Pawn%20Object%20Reference) (ImpactPoint :type Vector) (KnockbackValue :type Float%20%28double-precision%29) (KnockupValue :type Float%20%28double-precision%29))
  (return (+ (* (Math|Vector|MakeVector 0.0 0.0 1.0) KnockupValue) (* (Math|Vector|Normalize (Math|Vector|ProjectVectorontoPlane (Math|Vector|Lerp(Vector) (* (Math|Vector|Normalize (- (Transformation|GetWorldLocation (Variables|Default|GetBody)) ImpactPoint)) -1.0) (Math|Vector|GetReflectionVector (Math|Vector|Normalize (Transformation|GetVelocity InputPin)) (Math|Vector|Normalize (- (Transformation|GetWorldLocation (Variables|Default|GetBody)) ImpactPoint))) (Variables|UserKnobs|GetDirectionAlpha)) (Transformation|GetUpVector (Variables|Default|GetTrigger)))) KnockbackValue))))

(fn DisableBumper
  (Variables|Default|SetbIsDisabled true)
  (Collision|SetCollisionEnabled (Variables|Default|GetTrigger) "NoCollision"))

(fn GenericKnockbackActions
  (params (Pawn :type Pawn%20Object%20Reference) (ImpactPoint :type Vector))
  (bind (_vehicleknockbackvalue _pawnknockbackvalue _vehicleknockupvalue _pawnknockupvalue) (CallFunction|GetKnockbackValue))
  (LocalVariables|Setlocal%20Pawn Pawn)
  (Utilities|FlowControl|Sequence
    (:then_0
      ;[1 1 1] Figure out which knockback values to use
      (Utilities|Casting|CastToFortPlayerPawnAthena (LocalVariables|Getlocal%20Pawn)
        (:then
          (LocalVariables|Setlocal%20Knockback _pawnknockbackvalue)
          (LocalVariables|Setlocal%20Knockup _pawnknockupvalue))
        (:CastFailed
          (LocalVariables|Setlocal%20Knockback _vehicleknockbackvalue)
          (LocalVariables|Setlocal%20Knockup _vehicleknockupvalue))) ; is thje object a player pawn? If so, use player values, \nelse use Vehicle Values
    )
    (:then_1
      (Networking|CanExecuteCosmeticEvents
        (:True
          (Audio|SpawnSoundAttached (Variables|Sound|GetPushback%20Sound) (Variables|/Script/Engine.Actor|GetRootComponent (LocalVariables|Getlocal%20Pawn))))))
    (:then_2
      (return (CallFunction|CalculateLaunchVector (LocalVariables|Getlocal%20Pawn) ImpactPoint (LocalVariables|Getlocal%20Knockback) (LocalVariables|Getlocal%20Knockup))))))

(fn GetKnockbackValue
  (bind (_vehicleimpulsemagniture_4_3f79b7c3469bc952924faa8bc00e2eb7 _pawnimpulsemagnitude_6_89dcc04642398d0b9d2e1ba2823ac12b) (Utilities|Struct|BreakSpeedBoostImpulsePreset (Utilities|Array|Get(aref) (Variables|UserKnobs|GetKnockbackValues) (Variables|UserKnobs|GetKnockbackValueIndex))))
  (bind (_vehicleimpulsemagniture_4_3f79b7c3469bc952924faa8bc00e2eb7_2 _pawnimpulsemagnitude_6_89dcc04642398d0b9d2e1ba2823ac12b_2) (Utilities|Struct|BreakSpeedBoostImpulsePreset (Utilities|Array|Get(aref) (select (== (Variables|UserKnobs|GetSideBounceLiftValueIndex) 0) (Variables|UserKnobs|GetKnockupValues) (Variables|UserKnobs|GetSideBounceLiftValues)) (select (== (Variables|UserKnobs|GetSideBounceLiftValueIndex) 0) (Variables|UserKnobs|GetKnockbackValueIndex) (Variables|UserKnobs|GetSideBounceLiftValueIndex)))))
  (bind _sidebounceliftvalueindex (Variables|UserKnobs|GetSideBounceLiftValueIndex))
  (bind _returnvalue (== _sidebounceliftvalueindex 0))
  (bind _knockupvalues (Variables|UserKnobs|GetKnockupValues))
  (bind _sidebounceliftvalues (Variables|UserKnobs|GetSideBounceLiftValues))
  (bind _returnvalue_2 (select _returnvalue _knockupvalues _sidebounceliftvalues))
  (bind _knockbackvalueindex (Variables|UserKnobs|GetKnockbackValueIndex))
  (bind _returnvalue_3 (select _returnvalue _knockbackvalueindex _sidebounceliftvalueindex))
  (bind _knockbackvalues (Variables|UserKnobs|GetKnockbackValues))
  (bind _knockbackvalueindex_2 (Variables|UserKnobs|GetKnockbackValueIndex))
  (if (and (Utilities|Array|IsValidIndex _knockbackvalues _knockbackvalueindex_2) (Utilities|Array|IsValidIndex _returnvalue_2 _returnvalue_3))
    (return _vehicleimpulsemagniture_4_3f79b7c3469bc952924faa8bc00e2eb7 _pawnimpulsemagnitude_6_89dcc04642398d0b9d2e1ba2823ac12b _vehicleimpulsemagniture_4_3f79b7c3469bc952924faa8bc00e2eb7_2 _pawnimpulsemagnitude_6_89dcc04642398d0b9d2e1ba2823ac12b_2)
    (else
      (return 0.0 0.0 0.0 0.0))))

(fn GetOverrideMeshMaterial
  (params (MeshComp :type Static%20Mesh%20Component%20Object%20Reference) (MatIdx :type Integer))
  (if (== MatIdx 0)
    (if (Math|Boolean|ORBoolean (Utilities|Equal(Object) (Variables|Default|GetBody) MeshComp) (Utilities|Equal(Object) (Variables|Default|GetOC_PinballBumper_02) MeshComp) (Utilities|Equal(Object) (Variables|Default|GetOC_PinballBumper_03) MeshComp))
      (bind _material_instance_dynamic (Variables|Default|GetMaterial%20Instance%20Dynamic)
        (:then
          (return _material_instance_dynamic))
        (:else
          (return (Variables|Default|GetParentMaterial)))))
    (else
      (return 0))))

(fn IsTargetableByLinearMover
  (return true))

(fn MessageScore
  (return (Variables|UserOptions|Advanced|GetHUD%20Message) (Variables|UserOptions|Advanced|GetHUD%20MessageScoreColor) (Variables|UserOptions|Advanced|GetHUD%20MessageColor) (Variables|UserOptions|Advanced|GetResetHUD%20MessageScore)))

(fn OnFinishedAction
  (bind _resettime (Variables|UserKnobs|GetResetTime))
  (CallFunction|DisableBumper)
  (if (and (> _resettime 0.0) (Networking|HasAuthority))
    (Utilities|Time|SetTimerbyEvent (EventDispatchers|CreateEvent :!event Reset%20Bumper) _resettime)))

(fn OnRep_bIsDisabled
  (if (Variables|Default|GetbIsDisabled)
    (CallFunction|DisableFX)
    (else
      (CallFunction|ResetFX))))

(fn ResetBumper
  (Variables|Default|SetbIsDisabled false)
  (CallFunction|ResetCollision))

(fn ResetCollision
  (Collision|SetCollisionEnabled (Variables|Default|GetTrigger) (select (Variables|UserKnobs|GetbAllowSideBounce) "QueryOnly" "NoCollision"))
  (Collision|SetCollisionEnabled (Variables|Default|GetUpTrigger) (select (Variables|UserKnobs|GetbAllowUpBounce) "QueryOnly" "NoCollision")))

(fn ShouldInvisiblePropsBeAllowed
  (return false))

(fn UpdateParentMaterial
  (params (ParentMaterial :type Material%20Interface%20Object%20Reference))
  (CallFunction|SwitchToConstantMaterial)
  (bind _output_get (Variables|Default|SetParentMaterial ParentMaterial))
  (Utilities|IsValid _output_get
    (:Is%20Valid
      (Rendering|Material|SetMaterial (Variables|Default|GetOC_PinballBumper_02) 0 _output_get)
      (bind _returnvalue (Class|FortMaterialCacheManager|GetCache (Variables|Default|GetMaterial%20Cache%20Settings) (Variables|Default|GetParentMaterial)))
      (Variables|Default|SetMaterial%20Cache _returnvalue))))

(fn ConstructionScript)
