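; Gameplay ability for the Creative Pinball Bumper device that launches and damages whatever it hits.
; On commit, it runs server-forced target selection, then routes vehicles, players, creatures and physics objects to the bumper's launch functions using each hit's impact point.
; Players in vehicles and flying pawns are skipped, and creatures require the bumper's affects-creatures setting.
; It then adds a gameplay cue, applies effects at the bumper's damage level and notifies the bumper it finished.
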(event Ability|EventActivateAbility
  (bind _asbp_pinball_bumper (Utilities|Casting|CastToBP_PinballBumper (Ability|GetAvatarActorfromActorInfo))
    (:then
      (Variables|Default|SetOwningActor _asbp_pinball_bumper)
      (bind _returnvalue (Ability|CommitAbility))
      (if _returnvalue
        (bind (_asynctaskproxy _targetdata) (Ability|Tasks|DoTargetSelection :TaskInstanceName "None" :bForceTargetingOnServer true)
          (:Targeted
            (for _index (range (Ability|TargetData|GetDataCountfromTargetData _targetdata))
              (bind _blaunchedsomeone (CallFunction|LaunchTargets _targetdata _index))
              (Variables|Default|SetbLaunchedSomone (or _blaunchedsomeone (Variables|Default|GetbLaunchedSomone))))
            (CallFunction|ShowClientEffects)
            (CallFunction|DamageTargets :TargetData _targetdata :GameplayEffectLevel (Variables|/Game/Creative/Devices/Pinball_Bumper/BP_PinballBumper.BP_PinballBumper_C|GetDamageLevel (Variables|Default|GetOwningActor)))
            (Class|BPPinballBumper|OnFinishedAction (Variables|Default|GetOwningActor)))
          (:Cancelled
            _))
        (Ability|EndAbility)
        (else
          _)))
    (:CastFailed
      _)))

(fn DamageTargets
  (params (TargetData :type Gameplay%20Ability%20Target%20Data%20Handle%20Structure%20%28by%20ref%29) (ApplicationTag :type Gameplay%20Tag%20Structure) (GameplayEffectLevel :type Integer))
  (for _array_element (Ability|TargetData|GetAllActorsfromTargetData TargetData)
    (bind _asfort_athena_vehicle (Utilities|Casting|CastToFortAthenaVehicle _array_element)
      (:then
        (if (Utilities|IsArrayNonEmpty (Vehicle|GetAllPassengers _asfort_athena_vehicle))
          _))
      (:CastFailed
        _))
    (Ability|ApplyGameplayEffectContainer (Ability|TargetData|AbilityTargetDatafromActor _array_element) ApplicationTag GameplayEffectLevel)))

(fn LaunchPawn
  (params (Instigator :type BP%20Pinball%20Bumper%20Object%20Reference) (Pawn :type Fort%20Pawn%20Object%20Reference) (Hit :type Hit%20Result%20Structure%20%28by%20ref%29))
  (bind (_bblockinghit _binitialoverlap _time _distance _location _impactpoint) (Collision|BreakHitResult Hit))
  (Class|BPPinballBumper|LaunchPawn Instigator Pawn _impactpoint))

(fn LaunchPhysicsObject
  (params (Intigator :type BP%20Pinball%20Bumper%20Object%20Reference) (HitPhysicsObject :type Fort%20Physics%20Pawn%20Object%20Object%20Reference) (Hit :type Hit%20Result%20Structure))
  (bind (_bblockinghit _binitialoverlap _time _distance _location _impactpoint) (Collision|BreakHitResult Hit))
  (Class|BPPinballBumper|LaunchPhysicsObject Intigator HitPhysicsObject _impactpoint))

(fn LaunchVehicle
  (params (Instigator :type BP%20Pinball%20Bumper%20Object%20Reference) (Target :type Fort%20Athena%20Vehicle%20Object%20Reference) (Hit :type Hit%20Result%20Structure%20%28by%20ref%29))
  (bind (_bblockinghit _binitialoverlap _time _distance _location _impactpoint) (Collision|BreakHitResult Hit))
  (Class|BPPinballBumper|LaunchVehicle Instigator Target _impactpoint))

(fn LaunchTargets
  (params (TargetData :type Gameplay%20Ability%20Target%20Data%20Handle%20Structure%20%28by%20ref%29) (Index :type Integer))
  (LocalVariables|SetHitResult (Ability|TargetData|GetHitResultfromTargetData TargetData Index))
  (for _array_element (Ability|TargetData|GetActorsfromTargetData TargetData Index)
    (bind _asfort_athena_vehicle (Utilities|Casting|CastToFortAthenaVehicle _array_element)
      (:then
        (CallFunction|LaunchVehicle (Variables|Default|GetOwningActor) _asfort_athena_vehicle (LocalVariables|GetHitResult))
        (LocalVariables|SetbInternalLaunchedSomeone true))
      (:CastFailed
        (bind _asfort_player_pawn (Utilities|Casting|CastToFortPlayerPawn _array_element)
          (:then
            (if (and (not (Vehicle|IsinVehicle _asfort_player_pawn)) (not (AI|Components|NavMovement|IsFlying (Variables|/Script/Engine.Character|GetCharacterMovement _asfort_player_pawn))))
              (CallFunction|LaunchPawn (Variables|Default|GetOwningActor) _asfort_player_pawn (LocalVariables|GetHitResult))
              (LocalVariables|SetbInternalLaunchedSomeone true)))
          (:CastFailed
            (bind _asfort_pawn (Utilities|Casting|CastToFortPawn _array_element)
              (:then
                (if (and (Variables|/Game/Creative/Devices/Pinball_Bumper/BP_PinballBumper.BP_PinballBumper_C|GetbAffectsCreatures (Variables|Default|GetOwningActor)) (not (AI|Components|NavMovement|IsFlying (Variables|/Script/Engine.Character|GetCharacterMovement _asfort_pawn))))
                  (CallFunction|LaunchPawn (Variables|Default|GetOwningActor) _asfort_pawn (LocalVariables|GetHitResult))
                  (LocalVariables|SetbInternalLaunchedSomeone true)))
              (:CastFailed
                (bind _asfort_physics_pawn_object (Utilities|Casting|CastToFortPhysicsPawnObject _array_element))
                (CallFunction|LaunchPhysicsObject (Variables|Default|GetOwningActor) _asfort_physics_pawn_object (LocalVariables|GetHitResult))
                (LocalVariables|SetbInternalLaunchedSomeone true))))))))
  (return (LocalVariables|GetbInternalLaunchedSomeone)))

(fn ShowClientEffects
  (Ability|AddGameplayCueToOwner (Variables|Default|GetPinballSidewaysCue)))
