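; Gameplay ability for the Creative Pinball Bumper device that knocks targets upward.
; Its launch events for vehicles, pawns and physics objects each hand off to the matching upward-launch function on the instigating bumper.
; ShowClientEffects adds the upwards pinball gameplay cue to the ability owner for the visual feedback.
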(event EventLaunchVehicle
  (params (Instigator :type BP%20Pinball%20Bumper%20Object%20Reference) (Target :type Fort%20Athena%20Vehicle%20Object%20Reference) (Hit :type Hit%20Result%20Structure%20%28by%20ref%29))
  (Class|BPPinballBumper|LaunchVehicleUp Instigator Target))

(event EventLaunchPawn
  (params (Instigator :type BP%20Pinball%20Bumper%20Object%20Reference) (Pawn :type Fort%20Pawn%20Object%20Reference) (Hit :type Hit%20Result%20Structure%20%28by%20ref%29))
  (Class|BPPinballBumper|LaunchPlayerUp Instigator Pawn))

(event EventLaunchPhysicsObject
  (params (Intigator :type BP%20Pinball%20Bumper%20Object%20Reference) (HitPhysicsObject :type Fort%20Physics%20Pawn%20Object%20Object%20Reference) (Hit :type Hit%20Result%20Structure))
  (Class|BPPinballBumper|LaunchPhysicsObjectUp Intigator HitPhysicsObject))

(fn ShowClientEffects
  (Ability|AddGameplayCueToOwner (Variables|Default|GetPinballUpwardsCue)))
