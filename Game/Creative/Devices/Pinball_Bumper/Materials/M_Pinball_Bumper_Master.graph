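# Master material for the Creative Pinball Bumper device, with tintable diffuse, emissive glow and separate rubber surface settings.
# Emissive mask R selects bumper tint and glow, B adds inner glow, and spec mask G blends rubber against base metallic and roughness.
# GlowON blends resting emissive, picked by a static switch, into glow color, scaled separately on mobile.
# Attributes pass through Environment Effects, an optional lava edge gradient, then the Creative move tool function.
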Material /Game/Creative/Devices/Pinball_Bumper/Materials/M_Pinball_Bumper_Master
graph {
 properties {
  bAutomaticallySetUsageInEditor=false
  bUseMaterialAttributes=true
  bUsedWithInstancedStaticMeshes=true
  ditheredLODTransition=true
  pixelDepthOffsetMode=PDOM_Legacy
  usedShadingModels=Default Lit
 }
 nodes [
  n1 = Multiply(
   A=TextureSampleParameter2D() {parameterName=Diffuse, texture=/Game/Creative/Devices/Pinball_Bumper/Textures/T_Pinball_D}.RGB
   B=VectorParameter() {parameterName=Base_Color_Mult, defaultValue=<rgba 1 1 1 1>}.RGB
  )
  n2 = TextureSampleParameter2D() {parameterName=EmissiveMask, texture=/Game/Creative/Devices/Pinball_Bumper/Textures/T_Pinball_E}
  n3 = Multiply(A=VectorParameter() {parameterName=Bumper_Diffuse, defaultValue=<rgba 1 1 1 1>}.RGB, B=@n2.R)
  n4 = ScalarParameter() {parameterName=Metallic_Rubber}
  n5 = ScalarParameter() {parameterName=Metallic}
  n6 = TextureSampleParameter2D() {parameterName=Spec, samplerType=SAMPLERTYPE_Masks, texture=/Game/Creative/Devices/Pinball_Bumper/Textures/T_Pinball_S}
  n7 = ScalarParameter() {parameterName=Roughness_Rubber}
  n8 = ScalarParameter() {parameterName=Roughness}
  n9 = VectorParameter() {parameterName=Emi8ssive_Color_Bumper_Initial, defaultValue=<rgba 0 0 0 1>}
  n10 = Multiply(A=@n2.B, B=VectorParameter() {parameterName=Emissive_Color_Inner, defaultValue=<rgba 1 0 0 1>}.RGB)
  n11 = VectorParameter() {parameterName=Emissive_Color_Bumper, defaultValue=<rgba 0 0 0 1>}
  n12 = StaticSwitchParameter(True=Add(A=Multiply(A=@n9.RGB, B=@n2.R), B=@n10), False=Add(A=Multiply(A=@n11.RGB, B=@n2.R), B=@n10)) {parameterName=BumperGlowDefaultON}
  n13 = VectorParameter() {parameterName=Glow_Bumper_Color, defaultValue=<rgba 10 10 10 1>}
  n14 = Multiply(A=@n13.RGB, B=ScalarParameter() {defaultValue=0.25, parameterName=Mobile_Intensity_Multiplier})
  n15 = Add(A=Multiply(A=ShadingPathSwitch(Default=@n13.RGB, Mobile=@n14) {desc=Migrated from FeatureLevelSwitch}, B=@n2.R), B=@n10)
  n16 = ScalarParameter() {parameterName=GlowON}
  n17 = MaterialFunctionCall!(Environment FX=MakeMaterialAttributes!(
    BaseColor=LinearInterpolate(A=@n1, B=@n3, Alpha=@n2.R)
    Metallic=LinearInterpolate(A=@n4, B=@n5, Alpha=@n6.G)
    Specular=@n6.RGB
    Roughness=LinearInterpolate(A=@n7, B=@n8, Alpha=@n6.G)
    Anisotropy
    EmissiveColor=LinearInterpolate(A=@n12, B=@n15, Alpha=@n16)
    Opacity
    OpacityMask
    Normal=TextureSampleParameter2D() {parameterName=Normal, samplerType=SAMPLERTYPE_Normal, texture=/Game/Creative/Devices/Pinball_Bumper/Textures/T_Pinball_N}.RGB
    Tangent
    WorldPositionOffset
    SubsurfaceColor
    ClearCoat
    ClearCoatRoughness
    AmbientOcclusion
    Refraction
    CustomizedUVs_0
    CustomizedUVs_1
    CustomizedUVs_2
    CustomizedUVs_3
    CustomizedUVs_4
    CustomizedUVs_5
    CustomizedUVs_6
    CustomizedUVs_7
    PixelDepthOffset
    ShadingModel
    Displacement
   )) {materialFunction=/Game/Packages/Fortress_SharedMaterials/Base_Material_Functions/Environment_Effects}
  MaterialFunctionCall(MA In=StaticSwitchParameter(
    True=MaterialFunctionCall!(In=@n17.Result, UVs, TransitionGradient) {materialFunction=/Game/Environments/Sets/Vulcano/Materials/MF_LavaEdgeGradientSimple}.Result
    False=@n17.Result
   ) {parameterName=UseLavaEdgeGradient}) {materialFunction=&MF_CreativeMoveTool}
 ]
}
